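# Sound cue for demolition impacts on streetlights, using small-object attenuation and a four-voice concurrency limit.
# A quality-level node selects among four branches of decreasing complexity.
# The first two mix a random pick of two streetlight impacts with a briefly delayed, pitched-up hydrant metal hit that adds a secondary clank.
# The third keeps only the random impact and the fourth a single wave, all with pitch and volume modulation.
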SoundCue /Game/Sounds/Fort_Impact_Sounds/Demolition/StreetLight/Demo_StreetLight_Impact_Cue
root=QualityLevel_0
nodes [
 {id=Delay_0, children [WavePlayer_9], properties {delayMax=0.05, delayMin=0.02}}
 {id=Delay_2, children [WavePlayer_5], properties {delayMax=0.05, delayMin=0.02}}
 {id=Mixer_1, children [Modulator_2,Modulator_3], properties {inputVolume [1,1]}}
 {id=Mixer_2, children [Modulator_5,Modulator_7], properties {inputVolume [1,1]}}
 {id=Modulator_0, children [Random_0], properties {pitchMax=1.07, pitchMin=0.92}}
 {id=Modulator_10, children [Modulator_9], properties {volumeMax=1, volumeMin=0.9}}
 {id=Modulator_2, children [Random_2], properties {pitchMax=1.07, pitchMin=0.92}}
 {id=Modulator_3, children [Delay_0], properties {pitchMax=1.6, pitchMin=1.5}}
 {id=Modulator_4, children [Modulator_0], properties {volumeMax=1, volumeMin=0.9}}
 {id=Modulator_5, children [Random_1], properties {pitchMax=1.07, pitchMin=0.92}}
 {id=Modulator_6, children [Mixer_1], properties {volumeMax=0.7, volumeMin=0.7}}
 {id=Modulator_7, children [Delay_2], properties {pitchMax=1.6, pitchMin=1.5}}
 {id=Modulator_8, children [Mixer_2], properties {volumeMax=0.7, volumeMin=0.7}}
 {id=Modulator_9, children [WavePlayer_8], properties {pitchMax=1.07, pitchMin=0.92}}
 {id=QualityLevel_0, children [Modulator_6,Modulator_8,Modulator_4,Modulator_10]}
 {id=Random_0, children [WavePlayer_0,WavePlayer_1], properties {weights [1,1]}}
 {id=Random_1, children [WavePlayer_3,WavePlayer_4], properties {weights [1,1]}}
 {id=Random_2, children [WavePlayer_6,WavePlayer_7], properties {weights [1,1]}}
 {id=WavePlayer_0, properties {soundWaveAssetPtr=&/streetlight_impact_02}}
 {id=WavePlayer_1, properties {soundWaveAssetPtr=&/streetlight_impact_01}}
 {id=WavePlayer_3, properties {soundWaveAssetPtr=&streetlight_impact_02}}
 {id=WavePlayer_4, properties {soundWaveAssetPtr=&streetlight_impact_01}}
 {id=WavePlayer_5, properties {soundWaveAssetPtr=/Game/Sounds/Fort_Impact_Sounds/Demolition/Hydrant/metal_hydrant_hit_01}}
 {id=WavePlayer_6, properties {soundWaveAssetPtr=&streetlight_impact_02}}
 {id=WavePlayer_7, properties {soundWaveAssetPtr=&streetlight_impact_01}}
 {id=WavePlayer_8, properties {soundWaveAssetPtr=&streetlight_impact_02}}
 {id=WavePlayer_9, properties {soundWaveAssetPtr=&metal_hydrant_hit_01}}
]
properties {
 attenuationSettings=/Game/Sounds/Attenuation/Fort_Demolition_Hit_Small_Sized_Object_Att
 maxDistance=4000
 soundClassObject=&Fort_Impacts
 volumeMultiplier=0.34
 concurrencyOverrides {maxCount=4}
 concurrencySet [&CONC_Object_Impacts_FourMax]
}
